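; CoffeePuff weapon's hookshot ability, extending a parent that pulls hit targets: it fires instantly on activation.
; Clients suggest homing targets from the reticle's targeting component; successful pulls trigger camera, montage or feedback, and voice, timed to knockback.
; The server applies success or miss cooldowns and defers durability cost past knockback, with a once-only fallback on ability end.
; Damageable building hits cost separate durability, Mythic rarity changes damage level, and cooldown failures play feedback.
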(event Ability|EventActivateAbility
  ;[0 0.15 0.03639 0.5] ActivateAbility: Settings to instantly cast the hookshot as soon as the input is pressed.
  (Variables|Transient|SetbEarlyPrimaryQueued true)
  (Variables|Transient|SetbHookshotThrown true)
  (|CallParentFunction)
  (CallFunction|OnPrimaryTriggered))

(event EventClientHookshotComplete
  (params (bSuccessPull :type Boolean) (bPawnHitDead :type Boolean))
  ;[0.116 0.03839 0.15 0.5] ClientHookshotComplete: Jump to pull success after the knockback duration.
  (|CallParentFunction bSuccessPull bPawnHitDead)
  (if bSuccessPull
    (Utilities|FlowControl|Sequence
      (:then_0
        (Camera|PopCameraMode)
        (Camera|PushCameraMode (Variables|Camera|GetHookshotHitCameraModeClass))
        (Utilities|FlowControl|Delay (Utilities|Operators|Add (ScalableFloat|GetValueatLevel (Variables|Hotfix|GetHotfix_HookshotKnockbackPostDelay)) (ScalableFloat|GetValueatLevel (Variables|Hotfix|GetHotfix_HookshotKnockbackDuration)) 0.1))
        (Camera|PopCameraMode))
      (:then_1
        (if bPawnHitDead
          (Ability|Animation|MontageJumptoSection "PullSuccess")
          (else
            (Utilities|FlowControl|Delay (+ (ScalableFloat|GetValueatLevel (Variables|Hotfix|GetHotfix_HookshotKnockbackPostDelay)) (ScalableFloat|GetValueatLevel (Variables|Hotfix|GetHotfix_HookshotKnockbackDuration))))
            (bind _outfortplayerpawn (CallFunction|GetAvatarasValidFortPlayerPawn)
              (:ValidFortPlayerPawn
                (bind _returnvalue (Fortnite|GetFortPlayerControllerfromActor _outfortplayerpawn))
                (Utilities|IsValid _returnvalue
                  (:Is%20Valid
                    (Game|Feedback|ClientStartCameraShake :self _returnvalue :Shake "/CoffeePuffGameplay/Gameplay/Abilities/Hookshot/Camera/CoffeePuff_CameraShake_Hookshot_Pull.CoffeePuff_CameraShake_Hookshot_Pull_C" :Scale 1.0 :PlaySpace "CameraLocal")
                    (Game|Feedback|LocalPlayForceFeedback _returnvalue "/CoffeePuffGameplay/Gameplay/Abilities/Hookshot/ForceFeedback/FF_CoffeePuff_Hookshot_SuccessPull.FF_CoffeePuff_Hookshot_SuccessPull" false "None"))))))))
      (:then_2
        (if (not bPawnHitDead)
          (Utilities|FlowControl|Delay (Math|Float|Max(Float) (- (+ (ScalableFloat|GetValueatLevel (Variables|Hotfix|GetHotfix_HookshotKnockbackPostDelay)) (ScalableFloat|GetValueatLevel (Variables|Hotfix|GetHotfix_HookshotKnockbackDuration))) 0.175))))
        ;; "Get over here! "
        (bind _asb_coffee_puff_weapon (Utilities|Casting|CastToB_CoffeePuff_Weapon (Ability|GetCurrentSourceWeapon)))
        (Class|BCoffeePuffWeapon|PlayLocalHookshotVO _asb_coffee_puff_weapon)))))

(event Ability|EventOnEndAbility
  (params (bWasCancelled :type Boolean))
  ;[0.15 0.09534 0.004492 0.5] OnEndAbility: If hookshot is successful (it hit someone that got pulled), commit ability cost.
  (Utilities|FlowControl|Sequence
    (:then_0
      (Ability|HasAuthority
        (:True
          (if (Variables|Transient|GetbHookshotSuccessful)
            ;; Invoking durability cost here again in case the ability ends before the knockback durations are complete.
            (CallFunction|CommitAbilityDurabilityCost)))))
    (:then_1
      (|CallParentFunction bWasCancelled)
      (Variables|Transient|SetbDurabilityCostDeducted false))))

(event Default|EventApplyHookshot
  (bind _returnvalue (Utilities|Array|LastIndex (Projectile|GetProjectileDataHits)))
  (bind _returnvalue_2 (Ability|GetCurrentSourceWeapon))
  (bind (_bblockinghit _binitialoverlap _time _distance _location _impactpoint _normal _impactnormal _physmat _hitactor) (Collision|BreakHitResult (Utilities|Array|Get(aref) (Projectile|GetProjectileDataHits) _returnvalue)))
  (bind _hits (Projectile|GetProjectileDataHits))
  ;[0.02853 0.06123 0.15 0.5] ApplyHookshot: If hookshot is applied successfully, then apply the cost only only after the knockbacks are done.
  (|CallParentFunction)
  (Ability|HasAuthority
    (:True
      (Ability|ApplyGameplayEffectToOwner (select (Variables|Transient|GetbHookshotSuccessful) (Variables|Cooldown|GetHookshotSuccessCooldownGameplayEffect) (Variables|Cooldown|GetHookshotCooldownGameplayEffect)) (Level|GetHookshotDamageGameplayEffectLevel))
      (if (Variables|Transient|GetbHookshotSuccessful)
        (Utilities|FlowControl|Delay (Utilities|Operators|Add (ScalableFloat|GetValueatLevel (Variables|Hotfix|GetHotfix_HookshotKnockbackPostDelay)) (ScalableFloat|GetValueatLevel (Variables|Hotfix|GetHotfix_HookshotKnockbackDuration)) 0.15))
        (CallFunction|CommitAbilityDurabilityCost)
        (else
          (if (Utilities|Array|IsValidIndex _hits _returnvalue)
            (bind _asbuilding_actor (Utilities|Casting|CastToBuildingActor _hitactor))
            (if (Variables|/Script/Engine.Actor|GetbCanBeDamaged _asbuilding_actor)
              (Utilities|IsValid _returnvalue_2
                (:Is%20Valid
                  (FortWeapon|UseWeaponDurability(Scaled) _returnvalue_2 (ScalableFloat|GetValueatLevel (Variables|Hotfix|GetHotfix_BuildlingActorHitDurabilityCost))))))))))))

(event Custom|CommitAbilityDurabilityCost
  ;[0.15 0.06181 0.133 0.5] CommitAbilityDurabilityCost: Only deduct cost once per successful usage.
  (Ability|HasAuthority
    (:True
      (if (not (Variables|Transient|GetbDurabilityCostDeducted))
        (Variables|Transient|SetbDurabilityCostDeducted true)
        (Ability|CommitAbilityCost false)))))

(event EventSpawnHookshotProjectile
  ;[0.0218 0.1046 0.15 0.5] SpawnHookshotProjectile: Apply cast gameplay cue. Update the projectile reference on the weapon.
  (|CallParentFunction)
  (Ability|AddGameplayCueToOwner (Variables|GameplayCue|GetGameplayCueTagProjectileCast))
  (Ability|HasAuthority
    (:True
      (bind _asb_coffee_puff_weapon (Utilities|Casting|CastToB_CoffeePuff_Weapon (Ability|GetCurrentSourceWeapon)))
      (Class|BCoffeePuffWeapon|SetLatestHookshotProjectile _asb_coffee_puff_weapon (Utilities|Casting|CastToB_Prj_CoffeePuff_Hookshot (Variables|Transient|GetHookshotProjectile))))))

(fn ExecuteFailGameplayCue
  (params (Target :type Actor%20Object%20Reference))
  (GameplayCue|ExecuteGameplayCue%28Burst%29Local Target (Variables|GameplayCue|GetGameplayCueTagActivateFailed) (Ability|GameplayCue|MakeGameplayCueParameters 0.0)))

(fn GetClientSuggestedHomingTargetFromTargetingComponent
  (bind _returnvalue (Ability|GetCurrentSourceWeapon))
  ;[0.02665 0.1204 0.15 0.5] Overridden to return the target actor that the player's reticle is reacting to.
  (Utilities|FlowControl|Sequence
    (:then_0
      (Ability|HasAuthority
        (:False
          (Utilities|IsValid _returnvalue
            (:Is%20Valid
              (bind _returnvalue_2 (Actor|GetComponentbyClass _returnvalue "/AbilityTargetingFeedbackGameplay/Gameplay/B_AbilityTargetingFeedbackComponent_Parent.B_AbilityTargetingFeedbackComponent_Parent_C"))
              (Utilities|IsValid _returnvalue_2
                (:Is%20Valid
                  (Utilities|IsValid (.HitActor (Collision|BreakHitResult (Class|BAbilityTargetingFeedbackComponentParent|GetLatestTargetingTraceHitResult _returnvalue_2)))
                    (:Is%20Valid
                      (bind _bvalid (Homing|IsValidHomingTarget (.HitActor (Collision|BreakHitResult (Class|BAbilityTargetingFeedbackComponentParent|GetLatestTargetingTraceHitResult _returnvalue_2)))))
                      (if _bvalid
                        (return (.HitActor (Collision|BreakHitResult (Class|BAbilityTargetingFeedbackComponentParent|GetLatestTargetingTraceHitResult _returnvalue_2))))))))))))))
    (:then_1
      (return 0))))

(fn GetHookshotDamageGameplayEffectLevel
  (bind _returnvalue (Ability|GetCurrentSourceWeapon))
  ;[0 0.15 0.1392 0.5] GetHookshotDamageGameplayEffectLevel: Mythic variant deals damage at a different level.
  (Utilities|FlowControl|Sequence
    (:then_0
      (Utilities|IsValid _returnvalue
        (:Is%20Valid
          (return (Math|Integer|SelectInt 1 0 (GameplayTags|HasTag (Tags|GetWeaponDataGameplayTags _returnvalue) "(TagName=\"Rarity.Mythic\")"))))))
    (:then_1
      (return 0))))

(fn OverrideFailedReason
  (params (FailedReason :type Gameplay%20Tag%20Container%20Structure%20%28by%20ref%29))
  (Utilities|FlowControl|Sequence
    (:then_0
      (bind _outfortplayerpawn (CallFunction|GetAvatarasValidFortPlayerPawn)
        (:ValidFortPlayerPawn
          (if (GameplayTags|HasMatchingGameplayTag _outfortplayerpawn "(TagName=\"Granted.Cooldown.CoffeePuff.Hookshot\")")
            (GameplayTags|AddGameplayTag (LocalVariables|GetReturnFailedReasons) "(TagName=\"Activate.Failed.Cooldown\")" :!class /Script/GameplayTags.BlueprintGameplayTagLibrary)
            (Default|ExecuteFailGameplayCue (CallFunction|GetAvatarasFortPlayerPawn))))))
    (:then_1
      (return (LocalVariables|GetReturnFailedReasons)))))

(fn ShouldPlayFailedMontage
  (params (FailedReason :type Gameplay%20Tag%20Container%20Structure%20%28by%20ref%29))
  (return (GameplayTags|HasTag FailedReason "(TagName=\"Activate.Failed.Cooldown\")")))
