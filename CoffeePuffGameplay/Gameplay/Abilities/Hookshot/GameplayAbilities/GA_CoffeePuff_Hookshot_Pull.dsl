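; CoffeePuff Hookshot ability that pulls the target character toward the instigator via a custom, hotfix-gated movement mode extension with knockback before the pull.
; For player targets, the client applies the movement mode, with a server failsafe timer as backup.
; Gameplay events mark pull start, which blocks tagged abilities and plays weapon VO, and pull end, which finishes the launch, unblocks abilities and can grant same-team targets fall damage immunity.
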(event Ability|EventActivateAbility
  ;[0.0443 0.15 0.03206 0.5] Activate Ability: Cache whether pulling player is on the same team
  (|CallParentFunction)
  (bind _instigatorplayerpawn (Variables|Transient|GetInstigatorPlayerPawn)
    (:then
      (bind _targetcharacter (Variables|Transient|GetTargetCharacter)
        (:then
          (Variables|Transient|SetbInstigatorOnSameTeam (Team|OnSameTeam _targetcharacter _instigatorplayerpawn)))
        (:else)))
    (:else)))

(event EventDoLaunch
  (params (Vel :type Vector :default "0, 0, 0"))
  ;[0.15 0 0.1106 0.5] Do Launch: Overriden to bind to MME end via gameplay event on both server and client to call Launch Finished
  (if (ScalableFloat|GetValueasBool (Variables|Hotfix|GetHotfix_UseCustomMME))
    (bind _targetcharacter (Variables|Transient|GetTargetCharacter)
      (:then
        (Utilities|Casting|CastToFortPlayerPawn _targetcharacter
          (:then
            (Utilities|Casting|CastToFortPlayerController (Pawn|GetController _targetcharacter)
              (:then
                (Ability|HasAuthority
                  (:True
                    (bind _returnvalue (Utilities|Time|SetTimerbyEvent (EventDispatchers|CreateEvent :!event Custom|FailsafeServerApplyMME) (ScalableFloat|GetValueatLevel (Variables|Hotfix|GetHotfix_ServerMMEApplicationFailsafeTime))))
                    (Variables|Transient|SetFailsafeServerApplyMMETimer _returnvalue))
                  (:False
                    (CallFunction|ApplyPullMME)))
                (bind _asynctaskproxy (Ability|Tasks|WaitGameplayEvent "(TagName=\"Event.Ability.CoffeePuff.MME.PullStart\")" 0 true)
                  (:then
                    (Variables|Transient|SetWaitForPullStartAction _asynctaskproxy)
                    (bind _asynctaskproxy_2 (Ability|Tasks|WaitGameplayEvent "(TagName=\"Event.Ability.CoffeePuff.MME.End\")" 0 true)
                      (:then
                        (Variables|Transient|SetWaitForLaunchEndAction _asynctaskproxy_2))
                      (:EventReceived
                        (CallFunction|OnMMEEnd))))
                  (:EventReceived
                    (CallFunction|OnPullStart))))
              (:CastFailed
                (Utilities|Casting|CastToAIController (Pawn|GetController _targetcharacter)
                  (:then
                    _)
                  (:CastFailed
                    (|CallParentFunction Vel))))))
          (:CastFailed
            _)))
      (:else
        _))
    (else
      _)))

(event EventApplyPullMME
  ;[0.3573 0.09416 0.2881 0.5] Apply Pull MME: Overridden to apply custom MME that adds knockback before pull
  (if (ScalableFloat|GetValueasBool (Variables|Hotfix|GetHotfix_UseCustomMME))
    (bind _asfort_pawn (Utilities|Casting|CastToFortPawn (Variables|Transient|GetTargetCharacter)))
    (bind _returnvalue (Class|FortMovementModeExtCoffeePuffPull|TryActivateCoffeePuffPullMME _asfort_pawn "/CoffeePuffGameplay/Gameplay/Abilities/Hookshot/MME_CoffeePuff_Pull.MME_CoffeePuff_Pull_C" (Utilities|Struct|MakeFortMovementModeCoffeePuffPullCreationData (Variables|Transient|GetInstigatorPlayerPawn))))
    (Variables|Transient|SetPullMMEDataHandle _returnvalue)
    (else
      (|CallParentFunction))))

(event Custom|FailsafeServerApplyMME
  ;[0.15 0.06069 0.02261 0.5] Failsafe: Apply MME on server if it hasn't been applied a certain amount of time after attempting to apply on the client
  (CallFunction|ApplyPullMME))

(event Custom|On%20Pull%20Start
  ;[0.04709 0.15 0.1175 0.5] On Pull Start
  (CallFunction|BlockAbilitiesDuringPull true)
  (Utilities|Time|ClearandInvalidateTimerbyHandle (Variables|Transient|GetFailsafeServerApplyMMETimer))
  (Networking|CanExecuteCosmeticEvents
    (:True
      (bind _asb_coffee_puff_weapon (Utilities|Casting|CastToB_CoffeePuff_Weapon (Weapon|GetCurrentWeapon (Variables|Transient|GetInstigatorPlayerPawn))))
      (Class|BCoffeePuffWeapon|PlayLocalHookshotVO _asb_coffee_puff_weapon))))

(event Custom|On%20MME%20End
  ;[0.2656 0.1858 0.04402 0.5] MME End
  (CallFunction|LaunchFinished)
  (GameplayTasks|EndTask (Variables|Transient|GetWaitForLaunchEndAction))
  (CallFunction|BlockAbilitiesDuringPull false)
  (CallFunction|ApplyFallDamageImmunitytoSameTeam))

(event Ability|EventOnEndAbility
  (params (bWasCancelled :type Boolean))
  ;[0.3115 0.1163 0.02104 0.5] End Ability
  (|CallParentFunction bWasCancelled)
  (CallFunction|BlockAbilitiesDuringPull false)
  (Utilities|Time|ClearandInvalidateTimerbyHandle (Variables|Transient|GetFailsafeServerApplyMMETimer))
  (bind _waitforlaunchendaction (Variables|Transient|GetWaitForLaunchEndAction)
    (:then
      (GameplayTasks|EndTask _waitforlaunchendaction))
    (:else
      _))
  (bind _waitforpullstartaction (Variables|Transient|GetWaitForPullStartAction)
    (:then
      (GameplayTasks|EndTask _waitforpullstartaction))
    (:else)))

(fn ApplyFallDamageImmunityToSameTeam
  ;[0.4448 0.238 0.07739 0.5] Apply fall damamge immunity after pull if on the same team as the instigating player
  (if (and (Variables|Transient|GetbInstigatorOnSameTeam) (ScalableFloat|GetValueasBool (Variables|Hotfix|GetHotfix_ApplyFallDamageImmunityToSameTeam)))
    (Ability|ApplyGameplayEffectToOwner (Variables|Default|GetSameTeamFallDamageImmunityGE)) ; Apply fall damage immunity if on same team as instigator
  ))

(fn BlockAbilitiesDuringPull
  (params (Pulling :type Boolean))
  (bind _babilitiesblockedforpull (Variables|Transient|GetbAbilitiesBlockedForPull))
  (bind _pull_blockabilitieswithtag (Variables|Tags|GetPull_BlockAbilitiesWithTag))
  ;[0.175 0.4115 0.3877 0.5] Cancel and block abilities with tag during pull, and unblock abilities after pull.
  ; Used to allow abilities during knockback, but not durning pull. Example: Surface Swimming
  (bind _asfort_ability_system_component (Utilities|Casting|CastToFortAbilitySystemComponent (Ability|GetAbilitySystemComponentfromActorInfo)))
  (if Pulling
    (if (not _babilitiesblockedforpull)
      (Tags|CancelAbilitiesWithTags _asfort_ability_system_component _pull_blockabilitieswithtag)
      (Tags|BlockAbilitiesWithTags _asfort_ability_system_component _pull_blockabilitieswithtag)
      (Variables|Transient|SetbAbilitiesBlockedForPull true))
    (elif _babilitiesblockedforpull
      (Tags|UnBlockAbilitiesWithTags _asfort_ability_system_component _pull_blockabilitieswithtag false)
      (Variables|Transient|SetbAbilitiesBlockedForPull false))))
